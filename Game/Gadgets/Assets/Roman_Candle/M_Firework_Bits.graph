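# Translucent unlit particle sprite material for the glowing bits of the Roman Candle gadget's firework.
# An 8x8 SubUV husk-bits flipbook plays over particle relative time, and its green channel drives a multi-gradient color map, with particle red selecting the gradient row.
# A star glow texture blends that boosted color toward bright orange before a shared emissive adjustment function.
# Particle alpha multiplies the flipbook alpha for opacity.
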Material /Game/Gadgets/Assets/Roman_Candle/M_Firework_Bits
graph {
 properties {
  bEnableResponsiveAA=true
  bUsedWithNanite=true
  bUsedWithNiagaraSprites=true
  bUsedWithParticleSprites=true
  blendMode=BLEND_Translucent
  pixelDepthOffsetMode=PDOM_Legacy
  shadingModel=MSM_Unlit
  usedShadingModels=Unlit
 }
 nodes [
  n1 = MaterialFunctionCall!(
   Texture=TextureObject() {texture=/Game/Effects/Fort_Effects/Textures/Husk/T_Husk_Bits_SubUV}
   UVs
   SubImages=Constant2Vector() {g=8, r=8}
   Frame=Multiply(A=ParticleRelativeTime(), B=Constant() {r=64})
  ) {materialFunction=/Engine/Functions/Engine_MaterialFunctions03/Texturing/SubUV_Function}
  n2 = ParticleColor()
  n3 = MaterialFunctionCall!(
   Greyscale Value To Gradient Map=ComponentMask(None=@n1.RGB) {g=true}
   "Index (0 Based)"=@n2.R
   Gradient Texture=TextureObject() {texture=/Game/Effects/Fort_Effects/Textures/ColorMap/T_ColorMap01}
   Number Of Gradients In the Gradient Map=Constant() {r=8}
  ) {materialFunction=/Engine/Functions/Engine_MaterialFunctions02/Gradients/GradientMap_Multi}
  n4 = Constant3Vector() {constant=<rgba 0.905 0.48 0.0451 1>}
  MaterialFunctionCall(Emissive=LinearInterpolate(
    A=Multiply(A=@n3.RGB) {constB=60}
    B=Multiply(A=@n4) {constB=38}
    Alpha=TextureSample() {texture=/Game/Effects/Fort_Effects/Textures/Flares/T_GlowStar_5pt}.R
   )) {materialFunction=&MF_FX_EmissiveAdjustments}
  Multiply(A=@n2.A, B=@n1.Alpha)
 ]
}
